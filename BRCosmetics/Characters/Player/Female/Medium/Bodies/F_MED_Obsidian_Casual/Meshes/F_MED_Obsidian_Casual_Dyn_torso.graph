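# Physics asset for the torso of the Obsidian female medium Casual body, rigging secondary motion on its collar, hood, drawstrings and necklace.
# Capsules on the upper spine and clavicles form body volumes, linked by free angular constraints.
# Collar, hood and string chains start at small MSTR root bodies joined freely to spine_05, while the necklace attaches there directly.
# Chain joints limit swing and twist, with collars locked to a single swing axis.
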PhysicsAsset /BRCosmetics/Characters/Player/Female/Medium/Bodies/F_MED_Obsidian_Casual/Meshes/F_MED_Obsidian_Casual_Dyn_torso
bodies {
 clavicle_l [{shapeType=Capsule, length=10.5625, radius=7.1562, center=<5.6752 -0.4095 -2.6379>, rotation=<rot 90 0 0>}]
 clavicle_r [{shapeType=Capsule, length=10.5625, radius=7.1562, center=<-5.6752 0.4095 2.6379>, rotation=<rot 90 0 180>}]
 dyn_collar_l_01 [{shapeType=Capsule, length=4.4688, radius=2.125, center=<4.9664 0.0001 0>, rotation=<rot 90 0 0>}]
 dyn_collar_l_MSTR [{shapeType=Capsule, length=0.5, radius=0.5, rotation=<rot 90 0 0>}]
 dyn_collar_r_01 [{shapeType=Capsule, length=4.4688, radius=2.125, center=<4.9664 0.0001 0>, rotation=<rot 90 0 0>}]
 dyn_collar_r_MSTR [{shapeType=Capsule, length=0.5, radius=0.5, rotation=<rot 90 0 0>}]
 dyn_hood_bk_01 [{shapeType=Capsule, length=3.1562, radius=3, center=<2 0 2.2138>, rotation=<rot 90 0 0>}]
 dyn_hood_bk_02 [{shapeType=Capsule, length=3.1562, radius=3, center=<2 0 2.2138>, rotation=<rot 90 0 0>}]
 dyn_hood_bk_MSTR [{shapeType=Capsule, length=0.5, radius=0.5, rotation=<rot 90 0 0>}]
 dyn_hood_l_01 [{shapeType=Capsule, length=3.1562, radius=3, center=<2 0 2.2138>, rotation=<rot 90 0 0>}]
 dyn_hood_l_MSTR [{shapeType=Capsule, length=0.5, radius=0.5, rotation=<rot 90 0 0>}]
 dyn_hood_r_01 [{shapeType=Capsule, length=3.1562, radius=3, center=<2 0 2.2138>, rotation=<rot 90 0 0>}]
 dyn_hood_r_MSTR [{shapeType=Capsule, length=0.5, radius=0.5, rotation=<rot 90 0 0>}]
 dyn_necklace_l_01 [{shapeType=Capsule, length=5.0312, radius=3.25, center=<2 0 2.2138>, rotation=<rot 90 0 0>}]
 dyn_necklace_l_03 [{shapeType=Capsule, length=5.0312, radius=3.25, center=<2 0 2.2138>, rotation=<rot 90 0 0>}]
 dyn_string_l_01 [{shapeType=Capsule, length=2.5, radius=3, center=<2.0742 0 2.5>, rotation=<rot 90 0 0>}]
 dyn_string_l_02 [{shapeType=Capsule, length=2.5, radius=3, center=<2.0742 0 2.5>, rotation=<rot 90 0 0>}]
 dyn_string_l_03 [{shapeType=Capsule, length=2.25, radius=3.25, center=<2.0742 0 2.5>, rotation=<rot 90 0 0>}]
 dyn_string_l_04 [{shapeType=Capsule, length=2.5, radius=3, center=<2.0742 0 2.5>, rotation=<rot 90 0 0>}]
 dyn_string_l_05 [{shapeType=Capsule, length=2.5, radius=3, center=<2.0742 0 2.5>, rotation=<rot 90 0 0>}]
 dyn_string_l_06 [{shapeType=Capsule, length=2.5, radius=3, center=<2.0742 0 2.5>, rotation=<rot 90 0 0>}]
 dyn_string_l_07 [{shapeType=Capsule, length=2.5, radius=3, center=<2.0742 0 2.5>, rotation=<rot 90 0 0>}]
 dyn_string_l_08 [{shapeType=Capsule, length=2.5, radius=3, center=<2.0742 0 2.5>, rotation=<rot 90 0 0>}]
 dyn_string_l_MSTR [{shapeType=Capsule, length=0.5, radius=0.5, rotation=<rot 90 0 0>}]
 dyn_string_r_01 [{shapeType=Capsule, length=2.5, radius=3, center=<2.0742 0 2.5>, rotation=<rot 90 0 0>}]
 dyn_string_r_02 [{shapeType=Capsule, length=2.5, radius=3, center=<2.0742 0 2.5>, rotation=<rot 90 0 0>}]
 dyn_string_r_03 [{shapeType=Capsule, length=2.25, radius=3, center=<2.0742 0 2.5>, rotation=<rot 90 0 0>}]
 dyn_string_r_04 [{shapeType=Capsule, length=2.5, radius=3, center=<2.0742 0 2.5>, rotation=<rot 90 0 0>}]
 dyn_string_r_05 [{shapeType=Capsule, length=2.5, radius=3, center=<2.0742 0 2.5>, rotation=<rot 90 0 0>}]
 dyn_string_r_06 [{shapeType=Capsule, length=2.5, radius=3, center=<2.0742 0 2.5>, rotation=<rot 90 0 0>}]
 dyn_string_r_07 [{shapeType=Capsule, length=2.5, radius=3, center=<2.0742 0 2.5>, rotation=<rot 90 0 0>}]
 dyn_string_r_MSTR [{shapeType=Capsule, length=0.5, radius=0.5, rotation=<rot 90 0 0>}]
 spine_03 [{shapeType=Capsule, length=6.75, radius=8.9688, center=<0.8239 2.6103 0>, rotation=<rot 0 8.9469 0>}]
 spine_04 [{shapeType=Capsule, length=18.375, radius=8.9688, center=<0.6222 4.0913 0>}]
 spine_05 [
  {shapeType=Capsule, length=18.9375, radius=11.75, center=<-4.5317 -0.9927 0>, rotation=<rot 0 -7.3437 0>}
  {shapeType=Capsule, length=22.25, radius=7.75, center=<2.1143 -2.0038 0>}
 ]
}
constraints [
 {
  bone1Name=dyn_collar_l_01
  bone2Name=dyn_collar_l_MSTR
  swing1LimitDegrees=2
  swing1Motion=Locked
  swing2LimitDegrees=45
  twistLimitDegrees=2
  twistMotion=Locked
 }
 {
  bone1Name=dyn_collar_r_01
  bone2Name=dyn_collar_r_MSTR
  swing1LimitDegrees=2
  swing1Motion=Locked
  swing2LimitDegrees=45
  twistLimitDegrees=2
  twistMotion=Locked
 }
 {bone1Name=dyn_string_r_01, bone2Name=dyn_string_r_MSTR, swing1LimitDegrees=10, swing2LimitDegrees=10, twistLimitDegrees=5}
 {bone1Name=dyn_string_r_02, bone2Name=dyn_string_r_01, swing1LimitDegrees=10, swing2LimitDegrees=10, twistLimitDegrees=5}
 {bone1Name=dyn_string_r_03, bone2Name=dyn_string_r_02, swing1LimitDegrees=10, swing2LimitDegrees=10, twistLimitDegrees=5}
 {bone1Name=dyn_string_r_04, bone2Name=dyn_string_r_03, swing1LimitDegrees=10, swing2LimitDegrees=10, twistLimitDegrees=5}
 {bone1Name=dyn_string_r_05, bone2Name=dyn_string_r_04, swing1LimitDegrees=10, swing2LimitDegrees=10, twistLimitDegrees=5}
 {bone1Name=dyn_string_r_06, bone2Name=dyn_string_r_05, swing1LimitDegrees=10, swing2LimitDegrees=10, twistLimitDegrees=5}
 {bone1Name=dyn_string_r_07, bone2Name=dyn_string_r_06, swing1LimitDegrees=10, swing2LimitDegrees=10, twistLimitDegrees=5}
 {bone1Name=dyn_string_l_01, bone2Name=dyn_string_l_MSTR, swing1LimitDegrees=10, swing2LimitDegrees=10, twistLimitDegrees=5}
 {bone1Name=dyn_string_l_02, bone2Name=dyn_string_l_01, swing1LimitDegrees=10, swing2LimitDegrees=10, twistLimitDegrees=5}
 {bone1Name=dyn_string_l_03, bone2Name=dyn_string_l_02, swing1LimitDegrees=10, swing2LimitDegrees=10, twistLimitDegrees=5}
 {bone1Name=dyn_string_l_04, bone2Name=dyn_string_l_03, swing1LimitDegrees=10, swing2LimitDegrees=10, twistLimitDegrees=5}
 {bone1Name=dyn_string_l_05, bone2Name=dyn_string_l_04, swing1LimitDegrees=10, swing2LimitDegrees=10, twistLimitDegrees=5}
 {bone1Name=dyn_string_l_06, bone2Name=dyn_string_l_05, swing1LimitDegrees=10, swing2LimitDegrees=10, twistLimitDegrees=5}
 {bone1Name=dyn_string_l_07, bone2Name=dyn_string_l_06, swing1LimitDegrees=10, swing2LimitDegrees=10, twistLimitDegrees=5}
 {bone1Name=dyn_string_l_08, bone2Name=dyn_string_l_07, swing1LimitDegrees=10, swing2LimitDegrees=10, twistLimitDegrees=5}
 {
  bone1Name=dyn_necklace_l_03
  bone2Name=dyn_necklace_l_01
  swing1LimitDegrees=20
  swing2LimitDegrees=20
  twistLimitDegrees=5
  twistMotion=Locked
 }
 {bone1Name=dyn_hood_r_01, bone2Name=dyn_hood_r_MSTR, swing1LimitDegrees=5, swing2LimitDegrees=45, twistLimitDegrees=5}
 {bone1Name=dyn_hood_bk_01, bone2Name=dyn_hood_bk_MSTR, swing1LimitDegrees=5, swing2LimitDegrees=45, twistLimitDegrees=5}
 {bone1Name=dyn_hood_bk_02, bone2Name=dyn_hood_bk_01, swing1LimitDegrees=5, swing2LimitDegrees=45, twistLimitDegrees=5}
 {bone1Name=dyn_hood_l_01, bone2Name=dyn_hood_l_MSTR, swing1LimitDegrees=5, swing2LimitDegrees=45, twistLimitDegrees=5}
 {
  bone1Name=dyn_collar_l_MSTR
  bone2Name=spine_05
  swing1LimitDegrees=45
  swing1Motion=Free
  swing2LimitDegrees=45
  swing2Motion=Free
  twistLimitDegrees=45
  twistMotion=Free
 }
 {
  bone1Name=dyn_collar_r_MSTR
  bone2Name=spine_05
  swing1LimitDegrees=45
  swing1Motion=Free
  swing2LimitDegrees=45
  swing2Motion=Free
  twistLimitDegrees=45
  twistMotion=Free
 }
 {
  bone1Name=dyn_string_r_MSTR
  bone2Name=spine_05
  swing1LimitDegrees=45
  swing1Motion=Free
  swing2LimitDegrees=45
  swing2Motion=Free
  twistLimitDegrees=45
  twistMotion=Free
 }
 {
  bone1Name=dyn_string_l_MSTR
  bone2Name=spine_05
  swing1LimitDegrees=45
  swing1Motion=Free
  swing2LimitDegrees=45
  swing2Motion=Free
  twistLimitDegrees=45
  twistMotion=Free
 }
 {bone1Name=dyn_necklace_l_01, bone2Name=spine_05, swing1LimitDegrees=20, swing2LimitDegrees=20, twistLimitDegrees=5, twistMotion=Locked}
 {
  bone1Name=dyn_hood_r_MSTR
  bone2Name=spine_05
  swing1LimitDegrees=45
  swing1Motion=Free
  swing2LimitDegrees=45
  swing2Motion=Free
  twistLimitDegrees=45
  twistMotion=Free
 }
 {
  bone1Name=dyn_hood_bk_MSTR
  bone2Name=spine_05
  swing1LimitDegrees=45
  swing1Motion=Free
  swing2LimitDegrees=45
  swing2Motion=Free
  twistLimitDegrees=45
  twistMotion=Free
 }
 {
  bone1Name=dyn_hood_l_MSTR
  bone2Name=spine_05
  swing1LimitDegrees=45
  swing1Motion=Free
  swing2LimitDegrees=45
  swing2Motion=Free
  twistLimitDegrees=45
  twistMotion=Free
 }
 {
  bone1Name=spine_05
  bone2Name=spine_04
  swing1LimitDegrees=45
  swing1Motion=Free
  swing2LimitDegrees=45
  swing2Motion=Free
  twistLimitDegrees=45
  twistMotion=Free
 }
 {
  bone1Name=spine_04
  bone2Name=spine_03
  swing1LimitDegrees=45
  swing1Motion=Free
  swing2LimitDegrees=45
  swing2Motion=Free
  twistLimitDegrees=45
  twistMotion=Free
 }
 {
  bone1Name=clavicle_l
  bone2Name=spine_05
  swing1LimitDegrees=45
  swing1Motion=Free
  swing2LimitDegrees=45
  swing2Motion=Free
  twistLimitDegrees=45
  twistMotion=Free
 }
 {
  bone1Name=clavicle_r
  bone2Name=spine_05
  swing1LimitDegrees=45
  swing1Motion=Free
  swing2LimitDegrees=45
  swing2Motion=Free
  twistLimitDegrees=45
  twistMotion=Free
 }
]
