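# Physics asset for the dynamic front choker collar on the F_MED_Obsidian_Casual female player body.
# A spine_05 capsule anchors neck_02 and three collar surface capsules (top, mid, bottom), while neck_02 leads through dyn_collar_fr_MSTR to dyn_collar_fr_top.
# The collar top uses three capsules, two angled in opposite directions, and neck_02 and MSTR carry only tiny capsules as chain links.
# All constraints leave swing and twist free, so the listed angle limits do not restrict rotation.
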PhysicsAsset /Game/Characters/Player/Female/Medium/Bodies/F_MED_Obsidian_Casual/Meshes/F_MED_Obsidian_Casual_Dyn_choker
bodies {
 dyn_collar_fr_MSTR [{shapeType=Capsule, length=0.5, radius=0.5}]
 dyn_collar_fr_surface_bott [{shapeType=Capsule, length=4.1875, radius=3.8751, center=<-0.173 -3.9091 0>, rotation=<rot 90 -90 -91.4369>}]
 dyn_collar_fr_surface_mid [{shapeType=Capsule, length=5.2188, radius=3.8751, center=<0.0785 -3.8917 0>, rotation=<rot 90 0 -2.1537>}]
 dyn_collar_fr_surface_top [{shapeType=Capsule, length=4.2188, radius=3.8751, center=<0.1196 -3.895 0>, rotation=<rot 90 0 0>}]
 dyn_collar_fr_top [
  {shapeType=Capsule, length=12.204, radius=1.5, center=<-0.1386 1.4412 0>}
  {shapeType=Capsule, length=12.1562, radius=1.5, center=<0.3385 0.5123 -2.792>, rotation=<rot 0 0 45>}
  {shapeType=Capsule, length=12.1562, radius=1.5, center=<0.3385 0.5118 2.7926>, rotation=<rot 0 0 -45>}
 ]
 neck_02 [{shapeType=Capsule, length=0, radius=0.5}]
 spine_05 [{shapeType=Capsule, length=12.8031, radius=8.6478, center=<2.0911 -1.9204 -0.7007>, rotation=<rot -84.9975 -0.0001 -34.878>}]
}
constraints [
 {
  bone1Name=dyn_collar_fr_MSTR
  bone2Name=neck_02
  swing1LimitDegrees=45
  swing1Motion=Free
  swing2LimitDegrees=45
  swing2Motion=Free
  twistLimitDegrees=45
  twistMotion=Free
 }
 {
  bone1Name=dyn_collar_fr_top
  bone2Name=dyn_collar_fr_MSTR
  swing1LimitDegrees=45
  swing1Motion=Free
  swing2LimitDegrees=45
  swing2Motion=Free
  twistLimitDegrees=45
  twistMotion=Free
 }
 {
  bone1Name=dyn_collar_fr_surface_top
  bone2Name=spine_05
  swing1LimitDegrees=45
  swing1Motion=Free
  swing2LimitDegrees=45
  swing2Motion=Free
  twistLimitDegrees=45
  twistMotion=Free
 }
 {
  bone1Name=dyn_collar_fr_surface_mid
  bone2Name=spine_05
  swing1LimitDegrees=45
  swing1Motion=Free
  swing2LimitDegrees=45
  swing2Motion=Free
  twistLimitDegrees=45
  twistMotion=Free
 }
 {
  bone1Name=dyn_collar_fr_surface_bott
  bone2Name=spine_05
  swing1LimitDegrees=45
  swing1Motion=Free
  swing2LimitDegrees=45
  swing2Motion=Free
  twistLimitDegrees=45
  twistMotion=Free
 }
 {
  bone1Name=neck_02
  bone2Name=spine_05
  swing1LimitDegrees=45
  swing1Motion=Free
  swing2LimitDegrees=45
  swing2Motion=Free
  twistLimitDegrees=45
  twistMotion=Free
 }
]
