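# Post-process style AnimBP for the F_MED_Obsidian_Casual cosmetic body, adding deformation and secondary motion to the pose copied from its parent mesh.
# Control Rigs handle twists, sculpters and choker, alongside weighted constraints for collar, strings and zippers.
# RigidBody sims drive torso and jacket, then AnimDynamics springs forearm cloth and zippers.
# EventGraph updates gravity override and sim-space state via deprecated helpers, and bends the back collar by its distance to the right upper arm.
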AnimBlueprint /BRCosmetics/Characters/Player/Female/Medium/Bodies/F_MED_Obsidian_Casual/Meshes/F_MED_Obsidian_Casual_AnimBP
target_skeleton=&/F_MED_Obsidian_Casual_Skeleton
graphs [
 AnimGraph {nodes [
   n1 = "Constraint - dyn_collar_fr_MSTR (PPPPP)"!(
    ComponentPose
    Alpha=1
    ConstraintWeights_0=0.0759
    ConstraintWeights_1:float=0.1001
    ConstraintWeights_2:float=0.3704
    ConstraintWeights_3:float=0.1784
    ConstraintWeights_4:float=0.2751
   ) -> (Pose) {
    alphaBoolBlend {blendOption=Linear}
    boneToModify {boneName=dyn_collar_fr_MSTR}
    constraintSetup [
     {targetBone {boneName=C_jaw}}
     {targetBone {boneName=head}}
     {targetBone {boneName=neck_01}}
     {targetBone {boneName=spine_05}}
     {targetBone {boneName=neck_02}}
    ]
    constraintWeights [1,1,1,1,1]
   }
   n2 = "Constraint - dyn_collar_fr_surface_top (PPPPP)"!(
    ComponentPose=@n1
    Alpha=1
    ConstraintWeights_0=0.2422
    ConstraintWeights_1:float=0.1178
    ConstraintWeights_2:float=0.2786
    ConstraintWeights_3:float=0.053
    ConstraintWeights_4:float=0.3084
   ) -> (Pose) {
    alphaBoolBlend {blendOption=Linear}
    boneToModify {boneName=dyn_collar_fr_surface_top}
    constraintSetup [
     {targetBone {boneName=C_jaw}}
     {targetBone {boneName=head}}
     {targetBone {boneName=neck_01}}
     {targetBone {boneName=spine_05}}
     {targetBone {boneName=neck_02}}
    ]
    constraintWeights [1,1,1,1,1]
   }
   n3 = "Constraint - dyn_collar_fr_surface_mid (PPPPP)"!(
    ComponentPose=@n2
    Alpha=1
    ConstraintWeights_0=0.0759
    ConstraintWeights_1:float=0.1001
    ConstraintWeights_2:float=0.3704
    ConstraintWeights_3:float=0.1784
    ConstraintWeights_4:float=0.2751
   ) -> (Pose) {
    alphaBoolBlend {blendOption=Linear}
    boneToModify {boneName=dyn_collar_fr_surface_mid}
    constraintSetup [
     {targetBone {boneName=C_jaw}}
     {targetBone {boneName=head}}
     {targetBone {boneName=neck_01}}
     {targetBone {boneName=spine_05}}
     {targetBone {boneName=neck_02}}
    ]
    constraintWeights [1,1,1,1,1]
   }
   "Constraint - dyn_collar_fr_surface_bott (PPPP)"!(
    ComponentPose=@n3
    Alpha=1
    ConstraintWeights_0=0.2074
    ConstraintWeights_1:float=0.0517
    ConstraintWeights_2:float=0.388
    ConstraintWeights_3:float=0.3529
   ) -> (Pose) {
    alphaBoolBlend {blendOption=Linear}
    boneToModify {boneName=dyn_collar_fr_surface_bott}
    constraintSetup [{targetBone {boneName=neck_02}}, {targetBone {boneName=head}}, {targetBone {boneName=neck_01}}, {targetBone {boneName=spine_05}}]
    constraintWeights [1,1,1,1]
   }
   RigidBody(Alpha=0) {
    bClampLinearTranslationLimitToRefPose=true
    bOverrideWorldGravity=true
    overridePhysicsAsset=/Game/Characters/Player/Female/Medium/Bodies/F_MED_Obsidian_Casual/Meshes/F_MED_Obsidian_Casual_Dyn_choker
    simulationSpace=BaseBoneSpace
    alphaBoolBlend {blendOption=Linear}
    baseBoneRef {boneName=neck_02}
    componentAppliedLinearAccClamp=<0 0 0>
   }
   n4 = Control Rig(Source=Copy Pose From Mesh() {bUseAttachedParent=true}) {
    alphaBoolBlend {blendOption=Linear}
    controlRigAssetReference {blueprintRigClass=&/F_MED_Obsidian_Casual_Deform_CtrlRig_twists.F_MED_Obsidian_Casual_Deform_CtrlRig_twists_C}
   }
   n5 = Control Rig!(
    Source=@n4
    Alpha
    KneeRot_fr_l_114deg:Vector
    KneeTrans_fr_l_114deg:Vector="0.258,-3.591,0.009"
    KneeScale_fr_l_114deg:Vector="1,1,1"
    KneeScale_fr_l_130deg:Vector="1,1,1"
    KneeRot_fr_l_130deg:Vector="0,0,-51.2"
    KneeTrans_fr_l_130deg:Vector="0.258,-3.591,0.009"
   ) -> (Pose) {
    alphaBoolBlend {blendOption=Linear}
    controlRigAssetReference {blueprintRigClass=&/F_MED_Obsidian_Casual_Deform_CtrlRig_sculpters.F_MED_Obsidian_Casual_Deform_CtrlRig_sculpters_C}
   }
   n6 = "Transform (Modify) Bone - Bone: dyn_collar_r_MSTR"(
    ComponentPose=Local To Component(LocalPose=@n5)
    Alpha=Get collar_back_alpha() -> (collar_back_alpha:double)
    Translation="0,0,0"
    Rotation="-60,0,0"
    Scale="1,1,1"
   ) {
    rotationMode=BMM_Additive
    rotationSpace=BCS_BoneSpace
    alphaBoolBlend {blendOption=Linear}
    boneToModify {boneName=dyn_collar_r_MSTR}
    rotation {pitch=-60}
   }
   n7 = "Constraint - dyn_string_r_MSTR (PPPP)"!(
    ComponentPose=@n6
    Alpha=1
    ConstraintWeights_0=0.0351
    ConstraintWeights_1:float=0.1687
    ConstraintWeights_2:float=0.5939
    ConstraintWeights_3:float=0.2023
   ) -> (Pose) {
    alphaBoolBlend {blendOption=Linear}
    boneToModify {boneName=dyn_string_r_MSTR}
    constraintSetup [
     {targetBone {boneName=clavicle_r}}
     {targetBone {boneName=spine_04}}
     {targetBone {boneName=spine_05}}
     {targetBone {boneName=deform_pectoral_fr_r}}
    ]
    constraintWeights [1,1,1,1]
   }
   n8 = "Constraint - dyn_string_l_MSTR (PPP)"!(ComponentPose=@n7, Alpha=1, ConstraintWeights_0=0.2161, ConstraintWeights_1:float=0.1734, ConstraintWeights_2:float=0.6105) -> (Pose) {
    alphaBoolBlend {blendOption=Linear}
    boneToModify {boneName=dyn_string_l_MSTR}
    constraintSetup [{targetBone {boneName=deform_pectoral_fr_l}}, {targetBone {boneName=spine_04}}, {targetBone {boneName=spine_05}}]
    constraintWeights [1,1,1]
   }
   n9 = "Constraint - dyn_collar_fr_MSTR (PPPPP)"(
    ComponentPose=@n8
    Alpha=1
    ConstraintWeights_0=0.0759
    ConstraintWeights_1=0.3704
    ConstraintWeights_2=0.1001
    ConstraintWeights_3=0.1784
    ConstraintWeights_4=0.2751
   ) {
    alphaBoolBlend {blendOption=Linear}
    boneToModify {boneName=dyn_collar_fr_MSTR}
    constraintSetup [
     {targetBone {boneName=C_jaw}}
     {targetBone {boneName=neck_01}}
     {targetBone {boneName=head}}
     {targetBone {boneName=spine_05}}
     {targetBone {boneName=neck_02}}
    ]
    constraintWeights [1,1,1,1,1]
   }
   n10 = "Constraint - dyn_collar_fr_tgt_MSTR (PPPP)"!(
    ComponentPose=@n9
    Alpha=1
    ConstraintWeights_0=0.2074
    ConstraintWeights_1:float=0.388
    ConstraintWeights_2:float=0.0517
    ConstraintWeights_3:float=0.3529
   ) -> (Pose) {
    alphaBoolBlend {blendOption=Linear}
    boneToModify {boneName=dyn_collar_fr_tgt_MSTR}
    constraintSetup [{targetBone {boneName=neck_02}}, {targetBone {boneName=neck_01}}, {targetBone {boneName=head}}, {targetBone {boneName=spine_05}}]
    constraintWeights [1,1,1,1]
   }
   n11 = "Constraint - dyn_collar_fr_up_MSTR (PPPPP)"!(
    ComponentPose=@n10
    Alpha=1
    ConstraintWeights_0=0.0676
    ConstraintWeights_1:float=0.368
    ConstraintWeights_2:float=0.0976
    ConstraintWeights_3:float=0.2119
    ConstraintWeights_4:float=0.2548
   ) -> (Pose) {
    alphaBoolBlend {blendOption=Linear}
    boneToModify {boneName=dyn_collar_fr_up_MSTR}
    constraintSetup [
     {targetBone {boneName=C_jaw}}
     {targetBone {boneName=neck_01}}
     {targetBone {boneName=head}}
     {targetBone {boneName=spine_05}}
     {targetBone {boneName=neck_02}}
    ]
    constraintWeights [1,1,1,1,1]
   }
   n12 = Control Rig!(Source=Component To Local(ComponentPose=@n11), Alpha) -> (Pose) {
    alphaBoolBlend {blendOption=Linear}
    controlRigAssetReference {blueprintRigClass=&/F_MED_Obsidian_Casual_CtrlRig_choker.F_MED_Obsidian_Casual_CtrlRig_choker_C}
   }
   n13 = Get GravityOverride()
   n14 = RigidBody!(ComponentPose=Local To Component(LocalPose=@n12), Alpha, OverrideWorldGravity:Vector=@n13, ExternalForce) -> (Pose) {
    bClampLinearTranslationLimitToRefPose=true
    bOverrideWorldGravity=true
    overridePhysicsAsset=&/F_MED_Obsidian_Casual_Dyn_torso
    simulationSpace=BaseBoneSpace
    alphaBoolBlend {blendOption=Linear}
    baseBoneRef {boneName=spine_04}
    componentAppliedLinearAccClamp=<3000 3000 3000>
    componentLinearAccScale {z=0.5}
    componentLinearVelScale {z=0.5}
    overrideWorldGravity {z=-980}
    simSpaceSettings {
     maxAngularAcceleration=1000
     maxAngularVelocity=3
     maxLinearAcceleration=700
     maxLinearVelocity=500
     velocityScaleZ=0
     worldAlpha=1
     externalLinearDragV=<1 1>
    }
   }
   n15 = RigidBody(ComponentPose=@n14, OverrideWorldGravity=@n13) {
    bClampLinearTranslationLimitToRefPose=true
    bOverrideWorldGravity=true
    overridePhysicsAsset=&/F_MED_Obsidian_Casual_Dyn_jacket
    simulationSpace=BaseBoneSpace
    alphaBoolBlend {blendOption=Linear}
    baseBoneRef {boneName=spine_04}
    componentAppliedLinearAccClamp=<3000 3000 3000>
    componentLinearAccScale {z=0.5}
    componentLinearVelScale {z=0.5}
    overrideWorldGravity {z=-980}
    simSpaceSettings {
     maxAngularAcceleration=1000
     maxAngularVelocity=3
     maxLinearAcceleration=700
     maxLinearVelocity=500
     velocityScaleZ=0
     worldAlpha=1
     externalLinearDragV=<1 1>
    }
   }
   n16 = Anim Dynamics - Bone: dyn_forearm_cloth_l(ComponentPose=@n15, Alpha=1, ExternalForce="0,0,0") {
    angularDampingOverride=0.999
    angularSpringConstant=5000
    bAngularSpring=true
    bGravityOverrideInSimSpace=true
    bOverrideAngularDamping=true
    bUseGravityOverride=true
    simulationSpace=RootRelative
    alphaBoolBlend {blendOption=Linear}
    boundBone {boneName=dyn_forearm_cloth_l}
    gravityOverride {z=-980}
    physicsBodyDefinitions [{
      sphereCollisionRadius=0
      boundBone {boneName=dyn_forearm_cloth_l}
      boxExtents=<10 20 20>
      constraintSetup {angularLimitsMax=<5 5 5>, angularLimitsMin=<-5 -5 -5>, angularTarget=<1 0 0>}
      localJointOffset=<5 0 0>
     }]
   }
   n17 = Anim Dynamics - Bone: dyn_forearm_cloth_r(ComponentPose=@n16, Alpha=1, ExternalForce="0,0,0") {
    angularDampingOverride=0.999
    angularSpringConstant=5000
    bAngularSpring=true
    bGravityOverrideInSimSpace=true
    bOverrideAngularDamping=true
    bUseGravityOverride=true
    simulationSpace=RootRelative
    alphaBoolBlend {blendOption=Linear}
    boundBone {boneName=dyn_forearm_cloth_r}
    gravityOverride {z=-980}
    physicsBodyDefinitions [{
      sphereCollisionRadius=0
      boundBone {boneName=dyn_forearm_cloth_r}
      boxExtents=<10 20 20>
      constraintSetup {angularLimitsMax=<5 5 5>, angularLimitsMin=<-5 -5 -5>, angularTarget=<1 0 0>}
      localJointOffset=<5 0 0>
     }]
   }
   n18 = "Constraint - dyn_sleeve_zipper_l_MSTR (PPPP)"!(
    ComponentPose=@n17
    Alpha=1
    ConstraintWeights_0=0.0322
    ConstraintWeights_1:float=0.3016
    ConstraintWeights_2:float=0.0469
    ConstraintWeights_3:float=0.6192
   ) -> (Pose) {
    alphaBoolBlend {blendOption=Linear}
    boneToModify {boneName=dyn_sleeve_zipper_l_MSTR}
    constraintSetup [
     {targetBone {boneName=lowerarm_l}}
     {targetBone {boneName=lowerarm_twist_01_l}}
     {targetBone {boneName=lowerarm_twist_02_l}}
     {targetBone {boneName=dyn_forearm_cloth_l}}
    ]
    constraintWeights [1,1,1,1]
   }
   n19 = "Constraint - dyn_sleeve_zipper_r_MSTR (PPPP)"!(
    ComponentPose=@n18
    Alpha=1
    ConstraintWeights_0=0.0427
    ConstraintWeights_1:float=0.6232
    ConstraintWeights_2:float=0.0309
    ConstraintWeights_3:float=0.3032
   ) -> (Pose) {
    alphaBoolBlend {blendOption=Linear}
    boneToModify {boneName=dyn_sleeve_zipper_r_MSTR}
    constraintSetup [
     {targetBone {boneName=lowerarm_twist_02_r}}
     {targetBone {boneName=dyn_forearm_cloth_r}}
     {targetBone {boneName=lowerarm_r}}
     {targetBone {boneName=lowerarm_twist_01_r}}
    ]
    constraintWeights [1,1,1,1]
   }
   n20 = "Constraint - dyn_jacket_fr_r_zipper_MSTR (PPPP)"!(
    ComponentPose=@n19
    Alpha=1
    ConstraintWeights_0=0.8275
    ConstraintWeights_1:float=0.0936
    ConstraintWeights_2:float=0.0168
    ConstraintWeights_3:float=0.0621
   ) -> (Pose) {
    alphaBoolBlend {blendOption=Linear}
    boneToModify {boneName=dyn_jacket_fr_r_zipper_MSTR}
    constraintSetup [
     {targetBone {boneName=dyn_jacket_fr_r_02}}
     {targetBone {boneName=dyn_jacket_fr_r_01}}
     {targetBone {boneName=spine_04}}
     {targetBone {boneName=dyn_jacket_r_01}}
    ]
    constraintWeights [1,1,1,1]
   }
   n21 = Anim Dynamics - Bone: dyn_sleeve_zipper_l(ComponentPose=@n20, Alpha=1, ExternalForce="0,0,0") {
    angularDampingOverride=0.999
    angularSpringConstant=1000
    bAngularSpring=true
    bGravityOverrideInSimSpace=true
    bOverrideAngularDamping=true
    bUseGravityOverride=true
    simulationSpace=RootRelative
    alphaBoolBlend {blendOption=Linear}
    boundBone {boneName=dyn_sleeve_zipper_l}
    gravityOverride {z=-980}
    physicsBodyDefinitions [{
      sphereCollisionRadius=0
      boundBone {boneName=dyn_sleeve_zipper_l}
      boxExtents=<8 12 12>
      constraintSetup {angularLimitsMax=<0 0 60>, angularLimitsMin=<0 -89 -60>, angularTarget=<1 0 0>}
      localJointOffset=<5 0 0>
     }]
   }
   n22 = Anim Dynamics - Bone: dyn_sleeve_zipper_r(ComponentPose=@n21, Alpha=1, ExternalForce="0,0,0") {
    angularDampingOverride=0.999
    angularSpringConstant=1000
    bAngularSpring=true
    bGravityOverrideInSimSpace=true
    bOverrideAngularDamping=true
    bUseGravityOverride=true
    simulationSpace=RootRelative
    alphaBoolBlend {blendOption=Linear}
    boundBone {boneName=dyn_sleeve_zipper_r}
    gravityOverride {z=-980}
    physicsBodyDefinitions [{
      sphereCollisionRadius=0
      boundBone {boneName=dyn_sleeve_zipper_r}
      boxExtents=<8 12 12>
      constraintSetup {angularLimitsMax=<0 0 60>, angularLimitsMin=<0 -89 -60>, angularTarget=<1 0 0>}
      localJointOffset=<5 0 0>
     }]
   }
   n23 = Anim Dynamics - Bone: dyn_jacket_fr_r_zipper(ComponentPose=@n22, Alpha=1, ExternalForce="0,0,0") {
    angularDampingOverride=0.999
    angularSpringConstant=1000
    bAngularSpring=true
    bGravityOverrideInSimSpace=true
    bOverrideAngularDamping=true
    bUseGravityOverride=true
    simulationSpace=RootRelative
    alphaBoolBlend {blendOption=Linear}
    boundBone {boneName=dyn_jacket_fr_r_zipper}
    gravityOverride {z=-980}
    physicsBodyDefinitions [{
      sphereCollisionRadius=0
      boundBone {boneName=dyn_jacket_fr_r_zipper}
      boxExtents=<8 12 12>
      constraintSetup {angularLimitsMax=<0 89 60>, angularLimitsMin=<0 -60 -60>, angularTarget=<1 0 0>}
      localJointOffset=<5 0 0>
     }]
   }
   Output Pose(Result=Component To Local(ComponentPose=@n23)) {layerGroup=None}
  ]}
 EventGraph {nodes [
   n1 = Event BlueprintUpdateAnimation()
   > n2 = Sequence!(execute) -> (then_0, then_1, then_2:Exec)
   > n3 = Gravity Override State Function DEPRECATED(CharacterPartAnimInstance=Self-Reference(), GameTimeInSeconds=GetGameTimeInSeconds(), GravityOverrideParams=Get Gravity Override Params())
   > Set GravityOverride(GravityOverride=@n3)
   n4 = ComplexSimSpaceFunction DEPRECATED!(
    execute=@n2.then_1
    CharacterPartAnimInstance:Custom Character Part Anim Instance Object Reference=Self-Reference()
    InterpSpeedMax:double=9
    PawnSpeedMax:double=20
    RInterpErrorTolerance:double=1
    DeltaTime:double=@n1.DeltaTimeX
    DebugDraw:bool=false
    sim_space_rotation_in:Rotator=Get sim_space_rotation() -> (sim_space_rotation:Rotator)
    inComponentSpaceIn:bool=Get InComponentSpace() -> (InComponentSpace:bool)
    inComponentSpaceLastFrameIn:bool=Get InComponentSpaceLastFrame() -> (InComponentSpaceLastFrame:bool)
    lockedOnSimSpaceIn:bool=Get LockedOnSimSpace() -> (LockedOnSimSpace:bool)
    InterpSpeedIn:double=Get InterpSpeed() -> (InterpSpeed:double)
   ) -> (
    then
    sim_space_rotation:Rotator
    inComponentSpaceOut:bool
    inComponentSpaceLastFrameOut:bool
    lockedOnSimSpaceOut:bool
    InterpSpeedOut:double
   )
   > Set sim_space_rotation(sim_space_rotation:Rotator=@n4) -> (Output_Get:Rotator)
   > Set InComponentSpace(InComponentSpace:bool=@n4.inComponentSpaceOut) -> (Output_Get:bool)
   > Set InComponentSpaceLastFrame(InComponentSpaceLastFrame:bool=@n4.inComponentSpaceLastFrameOut) -> (Output_Get:bool)
   > Set LockedOnSimSpace(LockedOnSimSpace:bool=@n4.lockedOnSimSpaceOut) -> (Output_Get:bool)
   > Set InterpSpeed(InterpSpeed:double=@n4.InterpSpeedOut) -> (Output_Get:double)
   Set collar_back_alpha(
    execute=@n2.then_2
    collar_back_alpha:double=Get Distance Between Two Sockets!(
     Component:Skeletal Mesh Component Object Reference=Get Skel_Mesh()
     SocketOrBoneNameA:Name=dyn_collar_r_01
     SocketSpaceA:ERelativeTransformSpace Enum=RTS_World
     SocketOrBoneNameB:Name=upperarm_twist_01_r
     SocketSpaceB:ERelativeTransformSpace Enum=RTS_World
     bRemapRange:bool=true
     InRangeMin:float=5
     InRangeMax:float=14
     OutRangeMin:float=1
     OutRangeMax:float=0
    ) -> (ReturnValue:float) {is=title K2Node_CallFunction}
   ) -> (Output_Get:double)
   Event BlueprintInitializeAnimation()
   > Set Skel_Mesh(Skel_Mesh=GetPartSkeletalMeshComponent())
   GetOwnerSkeletalMeshComponent()
  ]}
]
variables [
 collar_back_alpha {type=number}
 Gravity Override Params {type=GravityOverrideParamsStruct}
 GravityOverride {type=Vector}
 InComponentSpace {type=boolean}
 InComponentSpaceLastFrame {type=boolean}
 InterpSpeed {type=number}
 LockedOnSimSpace {type=boolean}
 sim_space_rotation {type=Rotator}
 Skel_Mesh {type=SkeletalMeshComponent}
]
