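# Physics asset that provides cloth collision shapes for the F_MED_Obsidian_Casual female medium player body cosmetic.
# Three overlapping capsules on spine_03, spine_04 and spine_05 approximate the upper torso, growing wider toward the top.
# Two constraints chain neighboring spine bodies, with all swing and twist motion left free.
# The asset defines no shapes for the limbs, head or pelvis.
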PhysicsAsset /BRCosmetics/Characters/Player/Female/Medium/Bodies/F_MED_Obsidian_Casual/Meshes/F_MED_Obsidian_Casual_ClothCollision
bodies {
 spine_03 [{shapeType=Capsule, length=6.0312, radius=9.375, center=<0 2.4799 0>, rotation=<rot 90 0 0>}]
 spine_04 [{shapeType=Capsule, length=5.5625, radius=9.8438, center=<0 3.3084 0>, rotation=<rot 90 0 0>}]
 spine_05 [{shapeType=Capsule, length=4.25, radius=11.1562, center=<0 3.3084 0>, rotation=<rot 90 0 0>}]
}
constraints [
 {
  bone1Name=spine_04
  bone2Name=spine_03
  swing1LimitDegrees=45
  swing1Motion=Free
  swing2LimitDegrees=45
  swing2Motion=Free
  twistLimitDegrees=45
  twistMotion=Free
 }
 {
  bone1Name=spine_05
  bone2Name=spine_04
  swing1LimitDegrees=45
  swing1Motion=Free
  swing2LimitDegrees=45
  swing2Motion=Free
  twistLimitDegrees=45
  twistMotion=Free
 }
]
