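# Physics asset for the dynamic hair of the medium female Obsidian Casual player cosmetic.
# Five four-bone capsule chains (back center, back left and right, front left and right) hang from the head, each link constrained with identical swing and tight twist limits.
# A sphere-and-capsule head and an upper-spine capsule form the body shapes, joined by a fully free angular constraint.
# Middle front links carry an extra offset capsule, roughly mirrored left to right.
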PhysicsAsset /BRCosmetics/Characters/Player/Female/Medium/Bodies/F_MED_Obsidian_Casual/Meshes/Parts/F_MED_Obsidian_Casual_Dyn
bodies {
 dyn_hair_bk_01 [{shapeType=Capsule, length=3.7878, radius=7.1875, center=<2.5887 0.0002 7>, rotation=<rot 90 0 2.5>}]
 dyn_hair_bk_02 [{shapeType=Capsule, length=4.1003, radius=6.875, center=<2.5887 0.0002 7>, rotation=<rot 90 0 2.5>}]
 dyn_hair_bk_03 [{shapeType=Capsule, length=3.9753, radius=7, center=<2.5887 0.0002 7>, rotation=<rot 90 0 2.5>}]
 dyn_hair_bk_04 [{shapeType=Capsule, length=5.6941, radius=3, center=<1.8176 0 3>, rotation=<rot 0 0 90>}]
 dyn_hair_bk_l_01 [{shapeType=Capsule, length=3.6316, radius=7.3438, center=<2.5887 0.0002 7>, rotation=<rot 90 0 2.5>}]
 dyn_hair_bk_l_02 [{shapeType=Capsule, length=3.4441, radius=7.5312, center=<2.5887 0.0002 7>, rotation=<rot 90 0 2.5>}]
 dyn_hair_bk_l_03 [{shapeType=Capsule, length=4.2566, radius=6.7188, center=<2.5887 0.0002 7>, rotation=<rot 90 0 2.5>}]
 dyn_hair_bk_l_04 [{shapeType=Capsule, length=5.6941, radius=3, center=<1.8176 0 3>, rotation=<rot 0 0 90>}]
 dyn_hair_bk_r_01 [{shapeType=Capsule, length=3.6316, radius=7.3438, center=<2.5887 0.0002 7>, rotation=<rot 90 0 2.5>}]
 dyn_hair_bk_r_02 [{shapeType=Capsule, length=3.4441, radius=7.5312, center=<2.5887 0.0002 7>, rotation=<rot 90 0 2.5>}]
 dyn_hair_bk_r_03 [{shapeType=Capsule, length=4.2566, radius=6.7188, center=<2.5887 0.0002 7>, rotation=<rot 90 0 2.5>}]
 dyn_hair_bk_r_04 [{shapeType=Capsule, length=5.6941, radius=3, center=<1.8176 0 3>, rotation=<rot 0 0 90>}]
 dyn_hair_fr_l_01 [{shapeType=Capsule, length=3.9441, radius=7.0312, center=<2.5887 0.0002 7>, rotation=<rot 90 0 2.5>}]
 dyn_hair_fr_l_02 [
  {shapeType=Capsule, length=3.6941, radius=7.2812, center=<2.5887 0.0002 7>, rotation=<rot 90 0 2.5>}
  {shapeType=Capsule, length=6.5938, radius=1.5312, center=<2.8676 3.3152 -0.49>, rotation=<rot 90 1.8984 1.8984>}
 ]
 dyn_hair_fr_l_03 [
  {shapeType=Capsule, length=6.5691, radius=7.0625, center=<2.6252 -0.0014 6.7912>, rotation=<rot 84.9889 177.4999 -179.9999>}
  {shapeType=Capsule, length=4.3125, radius=3.8125, center=<2.8676 3.3306 2.0959>, rotation=<rot 90 -1.2659 -1.2659>}
 ]
 dyn_hair_fr_l_04 [{shapeType=Capsule, length=5.6941, radius=3, center=<1.8176 0 3>, rotation=<rot 0 0 90>}]
 dyn_hair_fr_r_01 [{shapeType=Capsule, length=3.9441, radius=7.0312, center=<2.5887 0.0002 7>, rotation=<rot 90 0 2.5>}]
 dyn_hair_fr_r_02 [
  {shapeType=Capsule, length=3.6941, radius=7.2812, center=<2.5887 0.0002 7>, rotation=<rot 90 0 2.5>}
  {shapeType=Capsule, length=6.5938, radius=1.5312, center=<2.868 -3.6077 -0.812>, rotation=<rot 90 -1.2659 -1.2659>}
 ]
 dyn_hair_fr_r_03 [
  {shapeType=Capsule, length=4.5066, radius=7.0625, center=<2.5887 0.0001 6.5335>, rotation=<rot 90 0 2.5>}
  {shapeType=Capsule, length=4.3125, radius=3.8125, center=<2.8666 -3.1021 2.2215>, rotation=<rot 90 -2.5305 -2.5305>}
 ]
 dyn_hair_fr_r_04 [{shapeType=Capsule, length=5.6941, radius=3, center=<1.8176 0 3>, rotation=<rot 0 0 90>}]
 head [
  {shapeType=Sphere, radius=6.8438, center=<3.7773 4.5534 0>}
  {shapeType=Capsule, length=6.3438, radius=6.5625, center=<2.8354 -0.4592 0>, rotation=<rot 90 0 0>}
  {shapeType=Capsule, length=6.2812, radius=6.625, center=<1.8261 2.6391 0>, rotation=<rot 90 0 0>}
  {shapeType=Capsule, length=3.0938, radius=3.625, center=<2.4302 0.2279 4.4837>, rotation=<rot 79.962 180 180>}
  {shapeType=Capsule, length=3.0938, radius=3.625, center=<2.4354 0.2279 -4.4846>, rotation=<rot -79.962 180 -180>}
 ]
 spine_05 [{shapeType=Capsule, length=3.4688, radius=8.5312, center=<7.9077 -0.8112 -0.0001>}]
}
constraints [
 {bone1Name=dyn_hair_fr_l_01, bone2Name=head, swing1LimitDegrees=20, swing2LimitDegrees=45, twistLimitDegrees=5}
 {bone1Name=dyn_hair_fr_l_02, bone2Name=dyn_hair_fr_l_01, swing1LimitDegrees=20, swing2LimitDegrees=45, twistLimitDegrees=5}
 {bone1Name=dyn_hair_fr_l_03, bone2Name=dyn_hair_fr_l_02, swing1LimitDegrees=20, swing2LimitDegrees=45, twistLimitDegrees=5}
 {bone1Name=dyn_hair_fr_l_04, bone2Name=dyn_hair_fr_l_03, swing1LimitDegrees=20, swing2LimitDegrees=45, twistLimitDegrees=5}
 {bone1Name=dyn_hair_bk_r_01, bone2Name=head, swing1LimitDegrees=20, swing2LimitDegrees=45, twistLimitDegrees=5}
 {bone1Name=dyn_hair_bk_r_02, bone2Name=dyn_hair_bk_r_01, swing1LimitDegrees=20, swing2LimitDegrees=45, twistLimitDegrees=5}
 {bone1Name=dyn_hair_bk_r_03, bone2Name=dyn_hair_bk_r_02, swing1LimitDegrees=20, swing2LimitDegrees=45, twistLimitDegrees=5}
 {bone1Name=dyn_hair_bk_r_04, bone2Name=dyn_hair_bk_r_03, swing1LimitDegrees=20, swing2LimitDegrees=45, twistLimitDegrees=5}
 {bone1Name=dyn_hair_fr_r_01, bone2Name=head, swing1LimitDegrees=20, swing2LimitDegrees=45, twistLimitDegrees=5}
 {bone1Name=dyn_hair_fr_r_02, bone2Name=dyn_hair_fr_r_01, swing1LimitDegrees=20, swing2LimitDegrees=45, twistLimitDegrees=5}
 {bone1Name=dyn_hair_fr_r_03, bone2Name=dyn_hair_fr_r_02, swing1LimitDegrees=20, swing2LimitDegrees=45, twistLimitDegrees=5}
 {bone1Name=dyn_hair_fr_r_04, bone2Name=dyn_hair_fr_r_03, swing1LimitDegrees=20, swing2LimitDegrees=45, twistLimitDegrees=5}
 {bone1Name=dyn_hair_bk_l_01, bone2Name=head, swing1LimitDegrees=20, swing2LimitDegrees=45, twistLimitDegrees=5}
 {bone1Name=dyn_hair_bk_l_02, bone2Name=dyn_hair_bk_l_01, swing1LimitDegrees=20, swing2LimitDegrees=45, twistLimitDegrees=5}
 {bone1Name=dyn_hair_bk_l_03, bone2Name=dyn_hair_bk_l_02, swing1LimitDegrees=20, swing2LimitDegrees=45, twistLimitDegrees=5}
 {bone1Name=dyn_hair_bk_l_04, bone2Name=dyn_hair_bk_l_03, swing1LimitDegrees=20, swing2LimitDegrees=45, twistLimitDegrees=5}
 {bone1Name=dyn_hair_bk_01, bone2Name=head, swing1LimitDegrees=20, swing2LimitDegrees=45, twistLimitDegrees=5}
 {bone1Name=dyn_hair_bk_02, bone2Name=dyn_hair_bk_01, swing1LimitDegrees=20, swing2LimitDegrees=45, twistLimitDegrees=5}
 {bone1Name=dyn_hair_bk_03, bone2Name=dyn_hair_bk_02, swing1LimitDegrees=20, swing2LimitDegrees=45, twistLimitDegrees=5}
 {bone1Name=dyn_hair_bk_04, bone2Name=dyn_hair_bk_03, swing1LimitDegrees=20, swing2LimitDegrees=45, twistLimitDegrees=5}
 {
  bone1Name=head
  bone2Name=spine_05
  swing1LimitDegrees=45
  swing1Motion=Free
  swing2LimitDegrees=45
  swing2Motion=Free
  twistLimitDegrees=45
  twistMotion=Free
 }
]
